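# Material function for the Holiday pumpkin prop that flickers emissive color in incoming material attributes, toggled by a static switch.
# Time plus a phase from object world position (X doubled, Z, no Y) varies flicker per placement, giving wind-like waves at higher frequency values.
# Two triangle waves, the second at a multiple of the first's rate, sum into the emissive multiplier.
# The Clamp on that sum is unconnected, so its min/max parameters do not affect output.
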MaterialFunction /Game/Environments/Helios/Props/Holiday/Holiday_TwoGridPumpkins_A/Materials/MF_PumpkinEmissiveFlicker
graph {nodes [
  n1 = ScalarParameter() {defaultValue=0.5, group=A_Emissive_Flicker, parameterName=Flicker1_Lerp Min}
  n2 = ScalarParameter() {defaultValue=2, group=A_Emissive_Flicker, parameterName=Flicker1_Max}
  n3 = ObjectPositionWS()
  n4 = ScalarParameter() {
   defaultValue=512
   desc="The frequency the world position will affect the flicker animation time loop. 512-1024 or higher values create wind waves effect between objects placed close together. \r\nlower values completelly randomizet the flicker"
   group=A_Emissive_Flicker
   parameterName=Wind_WP Wave Freq
  }
  n5 = MaterialFunctionCall(Color=Frac(None=Divide(A=@n3, B=@n4) {constB=512})) {materialFunction=&SplitComponents}
  n6 = Add(
   A=Multiply(A=Time(), B=ScalarParameter() {defaultValue=0.8, group=A_Emissive_Flicker, parameterName=Flicker1_Speed 1})
   B=Add(A=Add(A=@n5.R, B=@n5.R), B=@n5.B)
  )
  n7 = Frac(None=@n6)
  n8 = Abs(None=Subtract(A=Multiply(A=@n7) {constB=2}))
  n9 = Frac(None=Multiply(A=@n6, B=ScalarParameter() {defaultValue=4.137, group=A_Emissive_Flicker, parameterName=Flicker2_Speed}) {constB=14.1})
  n10 = Abs(None=Subtract(A=Multiply(A=@n9) {constB=2}))
  n11 = ScalarParameter() {defaultValue=1.3, group=A_Emissive_Flicker, parameterName=Flicker2_ Mult}
  n12 = NamedRerouteDeclaration(None=Add(A=LinearInterpolate(A=@n1, B=@n2, Alpha=@n8), B=Multiply(A=@n10, B=@n11) {constB=0.3})) {name=Flicker, nodeColor=<rgba 0 1 0.7656 1>}
  Clamp(
   None=@n12
   Min=ScalarParameter() {defaultValue=0.8, group=A_Emissive_Flicker, parameterName=Flicker_Clamp Min}
   Max=ScalarParameter() {defaultValue=2, group=A_Emissive_Flicker, parameterName=Flicke_Clamp Max}
  )
  n13 = GetMaterialAttributes(None=FunctionInput() {inputType=FunctionInput_MaterialAttributes}) {attributeGetTypes [#1]}
  n14 = Multiply(A=@n13.EmissiveColor, B=@n12)
  FunctionOutput(None=StaticSwitchParameter(
    True=SetMaterialAttributes!(MaterialAttributes=@n13.MaterialAttributes, EmissiveColor=@n14) {attributeSetTypes [#1]}
    False=@n13.MaterialAttributes
   ) {group=Emissive_Flicker, parameterName=Emissive Flicker, sortPriority=0})
 ]}
